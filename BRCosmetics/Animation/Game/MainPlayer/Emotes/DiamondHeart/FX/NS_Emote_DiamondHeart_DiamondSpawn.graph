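# One-shot Niagara sprite burst for the diamond spawn moment of the DiamondHeart emote.
# Layers a glow, round flash, outward triangle shards, multicolored glitter and curl-noise sparkles over two short smoke puffs with material-driven animation.
# Glitter, round flash and triangles follow User.Offset and dim by time of day.
# Glitter and triangle sizes scale with camera distance, and only the main Smoke emitter sets its own distance culling.
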NiagaraSystem /BRCosmetics/Animation/Game/MainPlayer/Emotes/DiamondHeart/FX/NS_Emote_DiamondHeart_DiamondSpawn
dependencies {
 usedDataInterfaces [&NiagaraDataInterfaceCurve, &NiagaraDataInterfaceColorCurve]
 usedDynamicInputs [&SimulationPosition, &ReturnNormalizedExecIndex]
 usedModules [&SystemState]
}
emitters {
 AirGlow {
  emitterUpdateScript [
   &EmitterState()
   &SpawnBurst_Instantaneous(Spawn Count=1, Use Spawn Probability=false)
  ]
  particleSpawnScript [
   &InitializeParticle(
    Color=<rgba 0.4323 0.4323 0.4323 0.2>
    Lifetime=0.25
    Sprite Size Mode=::NewEnumerator3
    Uniform Sprite Size=120
    UsePositionOffset=false
    Write Material Random=true
   ) {module=InitializeParticle001}
   &DynamicMaterialParameters(
    Index 0 Param 1=2
    Module.Param0WriteEnabled=true
    Module.Param1WriteEnabled=true
    Module.Param2WriteEnabled=true
    Module.Param3WriteEnabled=true
   )
   &CameraOffset(Camera Offset Amount=10)
  ]
  particleUpdateScript [
   &ParticleState()
   &ScaleSpriteSize(Scale Factor=&Vector2DFromFloat(Value=&FloatFromCurve(FloatCurve=DataInterface() {propertyValues {Curve {keys [<at 0 1 RCIM_Cubic>]}}}, Scale Curve=0.4)))
   &ScaleColor(
    Scale Alpha=&FloatFromCurve(FloatCurve=DataInterface() {propertyValues {Curve {keys [<at 0 0>,<at 0.25 0.75>,<at 1 0>]}}})
    ScaleA=true
    ScaleRGB=true
   ) {module=ScaleColor001}
  ]
  rendererClasses [&NiagaraSpriteRendererProperties]
  renderers [{rendererClass=&NiagaraSpriteRendererProperties}]
 }
 Glitter_Small {
  emitterUpdateScript [
   &EmitterState()
   &SpawnBurst_Instantaneous(Spawn Count=15, Use Spawn Probability=false, UseLoopCountLimit=false)
  ]
  particleSpawnScript [
   /SetVariables(Particles.SmallBigAlpha:NiagaraFloat=&UniformAOrBFloat(A=0)) {isSetParameters=true}
   &InitializeParticle(
    Color=&ColorFromCurve(
     ColorCurve=DataInterface() {propertyValues {
       AlphaCurve {keys [<at 0 1 RCIM_Cubic>]}
       BlueCurve {keys [<at 0 1 RCIM_Cubic>,<at 0.2987 1>,<at 0.3242 1>,<at 0.6155 1>,<at 0.6501 0.5954>,<at 0.8578 0.5642>,<at 0.8905 0.4196>,<at 1 0.4196>]}
       GreenCurve {keys [<at 0 0.6678 RCIM_Cubic>,<at 0.2987 0.6678>,<at 0.3242 0.3291>,<at 0.6155 0.3291>,<at 0.6501 1>,<at 0.8578 1>,<at 0.8905 1>,<at 1 1>]}
       RedCurve {keys [<at 0 0.31 RCIM_Cubic>,<at 0.2987 0.31>,<at 0.3242 0.3764>,<at 0.6155 0.3764>,<at 0.6501 0.4324>,<at 0.8578 0.4471>,<at 0.8905 0.9133>,<at 1 0.9133>]}
      }}
     CurveIndex=&RandomRangeFloat(Minimum=0, Maximum=1)
    )
    Lifetime Max=0.5
    Lifetime Min=0.3
    Lifetime Mode=::NewEnumerator1
    Mass Mode=::NewEnumerator3
    Position Offset=$User.Offset
    Sprite Size Mode=::NewEnumerator3
    Uniform Sprite Size=&Lerp_Float(
     A=&RandomRangeFloat(Minimum=5, Maximum=6)
     B=&RandomRangeFloat(Minimum=8, Maximum=10)
     Alpha=$Particles.SmallBigAlpha
    )
    UsePositionOffset=true
    Write Material Random=true
   )
   &ShapeLocation(MaskSpawnGroup=false, Offset Mode=::NewEnumerator0, Override Seed=false, Sphere Radius=20, UseOwnerScale=true)
   &AddVelocity(
    Velocity=$Particles.ShapeLocation.ShapeVector
    Velocity Speed Scale=&Lerp_Float(
     A=&RandomRangeFloat(Minimum=10, Maximum=20)
     B=&RandomRangeFloat(Minimum=15, Maximum=25)
     Alpha=$Particles.SmallBigAlpha
    )
   )
   &CameraOffset(Camera Offset Amount=6)
  ]
  particleUpdateScript [
   &ParticleState()
   &ScaleColor(
    Scale Alpha=&FloatFromCurve(
     FloatCurve=DataInterface() {propertyValues {Curve {keys [
         {arriveTangent=9.0725, interpMode=RCIM_Cubic, leaveTangent=9.0725, tangentMode=RCTM_User, time=0, value=0}
         {arriveTangent=-2.5249, interpMode=RCIM_Cubic, leaveTangent=-2.5249, tangentMode=RCTM_User, time=0.1408, value=0.9847}
         {arriveTangent=-0.0405, interpMode=RCIM_Cubic, leaveTangent=-0.0405, tangentMode=RCTM_User, time=1, value=0}
        ]}}}
     Scale Curve=5
    )
    Scale RGB=&Multiply_VectorByFloat(
     Vector=<1 1 1>
     Float=&FNTimeOfDayScalar(Night Scale=0.2, Morning Scale=0.5, Day Scale=0.5, Evening Scale=0.5)
    )
    ScaleA=true
    ScaleRGB=true
   ) {module=ScaleColor001}
   &Drag(Drag=&RandomRangeFloat(Minimum=8, Maximum=10), Use Linear Drag=true, Use Rotational Drag=true)
   &SolveForcesAndVelocity(Clamp Velocity=false, Limit Acceleration=false)
   &ScaleSpriteSize(
    Scale Sprite Size Mode=::NewEnumerator0
    Uniform Scale Factor=&FloatFromCurve(
     FloatCurve=DataInterface() {propertyValues {Curve {keys [{arriveTangent=1.6579, interpMode=RCIM_Cubic, leaveTangent=1.6579, tangentMode=RCTM_User, time=0, value=0}, <at 1 1 RCIM_Cubic>]}}}
     Scale Curve=&FN_LOD_Distance_Remap(Start Distance=100, End Distance=500, Minimum Output=0.75, Maximum Output=1.25, Clamp Output=true)
    )
   )
  ]
  rendererClasses [&NiagaraSpriteRendererProperties]
  renderers [{rendererClass=&NiagaraSpriteRendererProperties}]
 }
 RoundShape {
  emitterUpdateScript [
   &EmitterState()
   &SpawnBurst_Instantaneous(Use Spawn Probability=false, UseLoopCountLimit=false)
  ]
  particleSpawnScript [
   &InitializeParticle(
    Color=<rgba 1.7865 2.7595 3 0.25>
    Lifetime=0.4
    Mass Mode=::NewEnumerator3
    Position Offset=$User.Offset
    Sprite Size Mode=::NewEnumerator3
    Uniform Sprite Size=125
    UsePositionOffset=true
    Write Material Random=true
   )
   &ShapeLocation(MaskSpawnGroup=false, Non Uniform Scale=<0.5 1 1>, Override Seed=false, Sphere Radius=0, UseOwnerScale=true)
   &CameraOffset(Camera Offset Amount=4)
  ]
  particleUpdateScript [
   &ParticleState()
   &SpriteRotationRate(Rotation Rate=&FloatFromCurve(
     FloatCurve=DataInterface() {propertyValues {Curve {keys [
         {arriveTangent=1.4306, interpMode=RCIM_Cubic, leaveTangent=1.4306, tangentMode=RCTM_User, time=0.2, value=0}
         {arriveTangent=1.1179, interpMode=RCIM_Cubic, leaveTangent=1.1179, tangentMode=RCTM_User, time=0.4, value=0.6228}
         {arriveTangent=0.0224, interpMode=RCIM_Cubic, leaveTangent=0.0224, tangentMode=RCTM_User, time=1, value=1}
        ]}}}
     Scale Curve=250
    )) {enabled=false}
   &ScaleSpriteSize(
    Scale Sprite Size Mode=::NewEnumerator0
    Uniform Scale Factor=&FloatFromCurve(
     FloatCurve=DataInterface() {propertyValues {Curve {keys [{arriveTangent=2.6894, interpMode=RCIM_Cubic, leaveTangent=2.6894, tangentMode=RCTM_User, time=0, value=0}, <at 1 1 RCIM_Cubic>]}}}
     Scale Curve=0.7
    )
   )
   &ScaleColor(
    Scale Alpha=&FloatFromCurve(FloatCurve=DataInterface() {propertyValues {Curve {keys [
         {arriveTangent=9.0725, interpMode=RCIM_Cubic, leaveTangent=9.0725, tangentMode=RCTM_User, time=0, value=0.4}
         {arriveTangent=-0.0476, interpMode=RCIM_Cubic, leaveTangent=-0.0476, tangentMode=RCTM_User, time=0.1408, value=0.9847}
         {arriveTangent=-0.0405, interpMode=RCIM_Cubic, leaveTangent=-0.0405, tangentMode=RCTM_User, time=1, value=0}
        ]}}})
    Scale RGB=&Multiply_VectorByFloat(
     Vector=<1 1 1>
     Float=&FNTimeOfDayScalar(Night Scale=0.2, Morning Scale=0.5, Day Scale=0.5, Evening Scale=0.5)
    )
    ScaleA=true
    ScaleRGB=true
   ) {module=ScaleColor001}
  ]
  rendererClasses [&NiagaraSpriteRendererProperties]
  renderers [
   {rendererClass=&NiagaraSpriteRendererProperties}
   {rendererClass=&NiagaraSpriteRendererProperties, rendererIndex=1}
  ]
 }
 Smoke {
  emitterUpdateScript [
   &EmitterState(
    Enable Distance Culling=true
    Max Distance Response=::KillImmediately
    MaxDistance=7000
    Module.bUseMaxDistance=true
    Module.bUseMinDistance=false
    Scalability Mode=::NewEnumerator1
   )
   &SpawnBurst_Instantaneous(Spawn Count=&UniformRangedInt(Minimum=3, Maximum=5), Use Spawn Probability=false)
  ]
  particleSpawnScript [
   &InitializeParticle(
    Color=&MakeLinearColorFromVectorAndFloat(
     "Vector (RGB)"=&MakeVectorFromLinearColorRGB(Color=<rgba 0.1042 0.2395 0.3281 1>)
     "Float (Alpha)"=&UniformRangedFloat(Minimum=0.05, Maximum=0.1)
    )
    Lifetime=&UniformRangedFloat(Minimum=0.4, Maximum=0.6)
    Mass Mode=::NewEnumerator1
    Sprite Rotation Angle=&RandomRangeFloat(Minimum=0, Maximum=360)
    Sprite Rotation Mode=::NewEnumerator0
    Sprite Size=&Vector2DFromFloat(Value=&Multiply_Float(A=&UniformRangedFloat(Minimum=90, Maximum=120), B=1))
    Sprite Size Mode=::NewEnumerator4
    UsePositionOffset=false
    Write Material Random=true
   ) {module=InitializeParticle001}
   /SetVariables(Particles.InitSpriteRotRate=&UniformRangedFloat(Minimum=-3, Maximum=3)) {isSetParameters=true, module=SetVariables_#1}
   &AddVelocityInCone(
    Cone Angle=0
    Module.Use Velocity Falloff On Cone Axis=false
    Velocity Strength=&UniformRangedFloat(Minimum=50, Maximum=20)
   )
   &CameraOffset(Camera Offset Amount=&UniformRangedFloat(Minimum=10, Maximum=20))
  ]
  particleUpdateScript [
   &ParticleState()
   &SpriteRotationRate(Rotation Rate=&Multiply_Float(
     A=$Particles.InitSpriteRotRate
     B=&FloatFromCurve(FloatCurve=DataInterface() {propertyValues {Curve {keys [<at -0.0046 1.0113 RCIM_Cubic>,<at 1.0011 0.09 RCIM_Cubic>]}}})
    ))
   /Niagara/Modules/Update/SubUV/SubUVAnimation(Number Of Frames=64, SubUV Lookup Index=&Multiply_Float(A=$Particles.Age, B=0.66))
   &SolveForcesAndVelocity(Clamp Velocity=false, Limit Acceleration=false)
   &Color(
    Color=$Particles.Initial.Color
    Scale Alpha=&FloatFromCurve(FloatCurve=DataInterface() {propertyValues {Curve {keys [<at 0 0.5 RCIM_Cubic>,<at 1.0011 0.002 RCIM_Cubic>]}}})
   )
   &ScaleSpriteSize(Scale Factor=&Vector2DFromFloat(Value=&FloatFromCurve(FloatCurve=DataInterface() {propertyValues {Curve {keys [
          {arriveTangent=2.9149, interpMode=RCIM_Cubic, leaveTangent=2.9149, tangentMode=RCTM_User, time=-0.003, value=0}
          {arriveTangent=0.9127, interpMode=RCIM_Cubic, leaveTangent=0.9127, tangentMode=RCTM_User, time=0.23, value=0.63}
          <at 1 1>
         ]}}})))
   /Game/Effects/Niagara/Modules/Materials/MaterialSmoke(
    Erode:NiagaraFloat=&FloatFromCurve(
     FloatCurve=DataInterface() {propertyValues {
       bOptimizeLUT=false
       bUseLUT=false
       Curve {keys [
         {arriveTangent=-0.0652, interpMode=RCIM_Cubic, leaveTangent=-0.0652, tangentMode=RCTM_User, time=0.11, value=0.97}
         {arriveTangent=-5.7797, interpMode=RCIM_Cubic, leaveTangent=-5.7797, tangentMode=RCTM_User, time=0.39, value=0.66}
         {arriveTangent=-0.342, interpMode=RCIM_Cubic, leaveTangent=-0.342, tangentMode=RCTM_User, time=0.9954, value=-0.74}
        ]}
      }}
     Scale Curve=2
    )
    Shading:NiagaraFloat=&FloatFromCurve(FloatCurve=DataInterface() {propertyValues {
       bOptimizeLUT=false
       bUseLUT=false
       Curve {keys [
         <at 0.07 0.5 RCIM_Cubic>
         {arriveTangent=1.2205, interpMode=RCIM_Cubic, leaveTangent=1.2205, time=0.17, value=1.02}
         {interpMode=RCIM_Cubic, tangentMode=RCTM_User, time=0.3895, value=0.89}
         <at 0.8 0 RCIM_Cubic>
        ]}
      }})
   )
  ]
  rendererClasses [&NiagaraSpriteRendererProperties]
  renderers [{rendererClass=&NiagaraSpriteRendererProperties}]
 }
 Smoke001 {
  emitterUpdateScript [
   &EmitterState()
   &SpawnBurst_Instantaneous(Spawn Count=3, Use Spawn Probability=false) {module=SpawnBurst_Instantaneous001}
  ]
  particleSpawnScript [
   &InitializeParticle(
    Color=<rgba 0.5255 0.5255 0.5255 1>
    Lifetime=&UniformRangedFloat(Minimum=0.2, Maximum=0.6)
    Position Offset=<0 0 10>
    Position Offset Coordinate Space=::World
    Sprite Rotation Angle=&UniformRangedFloat(Minimum=0, Maximum=1)
    Sprite Rotation Mode=::NewEnumerator1
    Sprite Size Mode=::NewEnumerator3
    Uniform Sprite Size=&UniformRangedFloat(Minimum=15, Maximum=25)
    UsePositionOffset=true
    Write Material Random=true
   ) {module=InitializeParticle001}
   /SetVariables(Particles.InitSpriteRotRate=&UniformRangedFloat(Minimum=-1, Maximum=1)) {isSetParameters=true, module=SetVariables_#1}
   &AccelerationForce(Acceleration=&UniformRangedVector(Minimum=<0 0 -25>, Maximum=<0 0 -75>))
   &SphereLocation(
    MaskSpawnGroup=false
    Override Seed=false
    Sphere Radius=10
    Surface Only=true
    UseNonUniformScale=false
    UseOffset=true
    UseSphereOrientation=false
   )
   &AddVelocity(Velocity=<0 0 0>) {module=AddVelocity002}
   &AddVelocity(Rotation Coordinate Space=::World, Velocity=&RandomRangeVector(Minimum=<0 0 50>, Maximum=<0 0 200>)) {module=AddVelocity001}
   &ApplyInitialForces()
   &CameraOffset(Camera Offset Amount=&RandomRangeFloat(Minimum=0, Maximum=15))
   &AddVelocity(Velocity=&Multiply_Vector(
     A=$Output.SphereLocation.SphereNormal
     B=&VectorFromFloat(Value=&RandomRangeFloat(Minimum=40, Maximum=80))
    ))
  ]
  particleUpdateScript [
   &ParticleState()
   &SpriteRotationRate(Rotation Rate=&Multiply_Float(
     A=$Particles.InitSpriteRotRate
     B=&FloatFromCurve(
      FloatCurve=DataInterface() {propertyValues {Curve {keys [
          {arriveTangent=-1.9214, interpMode=RCIM_Cubic, leaveTangent=-1.9214, tangentMode=RCTM_User, time=-0.0046, value=1.0113}
          {arriveTangent=-0.2496, interpMode=RCIM_Cubic, leaveTangent=-0.2496, tangentMode=RCTM_User, time=0.9896, value=0.16}
         ]}}}
      CurveIndex=$Emitter.Age
     )
    ))
   &ScaleColor(
    Scale Alpha=&FloatFromCurve(FloatCurve=DataInterface() {propertyValues {Curve {keys [<at 0 0>,<at 0.2049 0.9765>,<at 0.7791 0.9107>,<at 1 0>]}}})
    ScaleA=true
    ScaleRGB=true
   ) {module=ScaleColor001}
   &Drag(Drag=8, Use Linear Drag=true, Use Rotational Drag=false)
   &SolveForcesAndVelocity(Clamp Velocity=false, Limit Acceleration=false)
   &ScaleSpriteSize(Scale Factor=&Vector2DFromFloat(Value=&FloatFromCurve(
      FloatCurve=DataInterface() {propertyValues {Curve {keys [
          {arriveTangent=5.2354, interpMode=RCIM_Cubic, leaveTangent=5.2354, tangentMode=RCTM_User, time=-0.0046, value=0.25}
          <at 0.1439 1.4923>
          <at 1.0011 3.4809 RCIM_Cubic>
         ]}}}
      Scale Curve=0.8
     )))
   &DynamicMaterialParameters(
    Index 0 Param 1=&FloatFromCurve(FloatCurve=DataInterface() {propertyValues {Curve {keys [<at 0 0 RCIM_Cubic>,<at 1.0011 0.2>]}}})
    Index 0 Param 2=&FloatFromCurve(FloatCurve=DataInterface() {propertyValues {Curve {keys [
         {arriveTangent=0.0453, interpMode=RCIM_Cubic, tangentMode=RCTM_Break, time=0, value=-1}
         {arriveTangent=0.0604, interpMode=RCIM_Cubic, leaveTangent=0.0604, tangentMode=RCTM_User, time=1, value=-1}
        ]}}})
    Index 0 Param 3=&FloatFromCurve(FloatCurve=DataInterface() {propertyValues {Curve {keys [{interpMode=RCIM_Cubic, leaveTangent=4.6779, time=0.2851, value=0}]}}})
    Index 1 Param 1=&FloatFromCurve(
     FloatCurve=DataInterface() {propertyValues {Curve {keys [<at 0 0>,<at 0.25 1>]}}}
     Scale Curve=&RandomRangeFloat(Minimum=2, Maximum=1)
    )
    Index 1 Param 2=&FloatFromCurve(
     FloatCurve=DataInterface() {propertyValues {Curve {keys [
         {interpMode=RCIM_Cubic, leaveTangent=3.7422, tangentMode=RCTM_Break, time=0, value=0.2}
         {interpMode=RCIM_Cubic, leaveTangent=0.9066, time=1, value=2}
        ]}}}
     Scale Curve=&RandomRangeFloat(Minimum=0.5, Maximum=1)
    )
    Index 1 Param 3=&FloatFromCurve(FloatCurve=DataInterface() {propertyValues {Curve {keys [
         {interpMode=RCIM_Cubic, leaveTangent=2.8445, tangentMode=RCTM_Break, time=0, value=0.2}
         {interpMode=RCIM_Cubic, leaveTangent=63.2978, time=0.25, value=1}
        ]}}})
    Index 1 Param 4=&FloatFromCurve(
     FloatCurve=DataInterface() {propertyValues {Curve {keys [<at 0 1>,<at 1 0.8>]}}}
     Scale Curve=&RandomRangeFloat(Minimum=0.5, Maximum=1)
    )
    Index 2 Param 1=&FloatFromCurve(FloatCurve=DataInterface() {propertyValues {Curve {keys [{interpMode=RCIM_Cubic, leaveTangent=3.2309, tangentMode=RCTM_Break, time=0.1, value=0}, <at 1 1>]}}})
    Index 2 Param 2=$Particles.MaterialRandom
    Index 2 Param 3=&FloatFromCurve(FloatCurve=DataInterface() {propertyValues {Curve {keys [
         {interpMode=RCIM_Cubic, tangentMode=RCTM_Break, time=0, value=1}
         {interpMode=RCIM_Cubic, leaveTangent=11.1233, tangentMode=RCTM_Break, time=1, value=2}
        ]}}})
    Index 2 Param 4=0.75
    Module.Param0WriteEnabled=true
    Module.Param0WriteEnabled1=true
    Module.Param0WriteEnabled2=true
    Module.Param1WriteEnabled=false
    Module.Param1WriteEnabled1=true
    Module.Param1WriteEnabled2=true
    Module.Param2WriteEnabled=true
    Module.Param2WriteEnabled1=true
    Module.Param2WriteEnabled2=true
    Module.Param3WriteEnabled=true
    Module.Param3WriteEnabled1=true
    Module.Param3WriteEnabled2=true
    Write Parameter Index 1=true
    Write Parameter Index 2=true
   )
  ]
  rendererClasses [&NiagaraSpriteRendererProperties]
  renderers [{rendererClass=&NiagaraSpriteRendererProperties}]
 }
 Sparkles {
  emitterUpdateScript [
   &EmitterState()
   &SpawnBurst_Instantaneous(Spawn Count=5, Use Spawn Probability=false)
  ]
  particleSpawnScript [
   &InitializeParticle(
    Color=&ColorFromCurve(
     ColorCurve=DataInterface() {propertyValues {
       AlphaCurve {keys [<at 0 1 RCIM_Cubic>]}
       BlueCurve {keys [<at 0 1 RCIM_Cubic>,<at 0.2987 1>,<at 0.3242 1>,<at 0.6155 1>,<at 0.6501 0.5954>,<at 0.8578 0.5642>,<at 0.8905 0.4197>,<at 1 0.4197>]}
       GreenCurve {keys [<at 0 0.6678 RCIM_Cubic>,<at 0.2987 0.6678>,<at 0.3242 0.3291>,<at 0.6155 0.3291>,<at 0.6501 1>,<at 0.8578 1>,<at 0.8905 1>,<at 1 1>]}
       RedCurve {keys [<at 0 0.31 RCIM_Cubic>,<at 0.2987 0.31>,<at 0.3242 0.3764>,<at 0.6155 0.3764>,<at 0.6501 0.4324>,<at 0.8578 0.4471>,<at 0.8905 0.9133>,<at 1 0.9133>]}
      }}
     CurveIndex=&RandomRangeFloat(Minimum=0, Maximum=1)
    )
    Lifetime=&UniformRangedFloat(Minimum=0.3, Maximum=1)
    Sprite Size Mode=::NewEnumerator3
    Uniform Sprite Size=&UniformRangedFloat(Minimum=10, Maximum=30)
    UsePositionOffset=false
    Write Material Random=true
   ) {module=InitializeParticle001}
   &AddVelocity(Velocity=&UniformRangedVector(Minimum=<-200 -200 -200>, Maximum=<200 200 200>))
   &AddVelocity(Rotation Coordinate Space=::World, Velocity=&RandomRangeVector(Minimum=<0 0 50>, Maximum=<0 0 100>)) {module=AddVelocity001}
   &ApplyInitialForces()
  ]
  particleUpdateScript [
   &ParticleState()
   &GravityForce(Gravity=<0 0 200>)
   /Niagara/Modules/Update/Forces/CurlNoiseForce(Noise Frequency=0.02, Noise Quality / Cost=::NewEnumerator0, Noise Strength=500, Pan Noise=false)
   &Drag(Drag=5, Use Linear Drag=true, Use Rotational Drag=false)
   &SolveForcesAndVelocity(Clamp Velocity=false, Limit Acceleration=false)
   &ScaleColor(
    Scale Alpha=&FloatFromCurve(FloatCurve=DataInterface() {propertyValues {Curve {keys [<at -0.0092 0.0144>,<at 0.0967 0.9767>,<at 1.0011 0.002>]}}})
    ScaleA=true
    ScaleRGB=true
   )
   &ScaleSpriteSize(Scale Factor=&Vector2DFromFloat(Value=&FloatFromCurve(FloatCurve=DataInterface() {propertyValues {Curve {keys [<at -0.0046 0>,<at 0.1746 0.9952>,<at 1.0011 0.002>]}}})))
  ]
  rendererClasses [&NiagaraSpriteRendererProperties]
  renderers [{rendererClass=&NiagaraSpriteRendererProperties}]
 }
 TriangleShapes {
  emitterUpdateScript [
   &EmitterState()
   &SpawnBurst_Instantaneous(Spawn Count=5, Use Spawn Probability=false, UseLoopCountLimit=false)
  ]
  particleSpawnScript [
   &InitializeParticle(
    Color=<rgba 0.6814 1.1776 1.5 2>
    Lifetime Max=0.3
    Lifetime Min=0.25
    Lifetime Mode=::NewEnumerator1
    Mass Mode=::NewEnumerator3
    Position Offset=$User.Offset
    Sprite Rotation Mode=::NewEnumerator2
    Sprite Size Mode=::NewEnumerator1
    Uniform Sprite Size Min=8
    UsePositionOffset=true
    Write Material Random=true
   )
   &ShapeLocation(MaskSpawnGroup=false, Offset Mode=::NewEnumerator0, Override Seed=false, Sphere Radius=3, UseOwnerScale=true)
   &AddVelocity(
    Velocity=$Particles.ShapeLocation.ShapeNormal
    Velocity Speed Scale=&RandomRangeFloat(Minimum=225, Maximum=275)
   )
   &CameraOffset(Camera Offset Amount=6)
  ]
  particleUpdateScript [
   &ParticleState()
   &SpriteRotationRate(Rotation Rate=&FloatFromCurve(
     FloatCurve=DataInterface() {propertyValues {Curve {keys [
         {arriveTangent=1.4306, interpMode=RCIM_Cubic, leaveTangent=1.4306, tangentMode=RCTM_User, time=0.2, value=0}
         {arriveTangent=0.1956, interpMode=RCIM_Cubic, leaveTangent=0.1956, tangentMode=RCTM_User, time=1, value=1}
        ]}}}
     Scale Curve=&RandomRangeFloat(Minimum=30, Maximum=50)
    ))
   &ScaleColor(
    Scale Alpha=&FloatFromCurve(FloatCurve=DataInterface() {propertyValues {Curve {keys [
         {arriveTangent=9.0725, interpMode=RCIM_Cubic, leaveTangent=9.0725, tangentMode=RCTM_User, time=0, value=0}
         {arriveTangent=-2.5249, interpMode=RCIM_Cubic, leaveTangent=-2.5249, tangentMode=RCTM_User, time=0.1408, value=0.9847}
         {arriveTangent=-0.0405, interpMode=RCIM_Cubic, leaveTangent=-0.0405, tangentMode=RCTM_User, time=1, value=0}
        ]}}})
    Scale RGB=&Multiply_VectorByFloat(
     Vector=<1 1 1>
     Float=&FNTimeOfDayScalar(Night Scale=0.2, Morning Scale=0.5, Day Scale=0.5, Evening Scale=0.5)
    )
    ScaleA=true
    ScaleRGB=true
   ) {module=ScaleColor001}
   /SetVariables(Particles.SubImageIndex=&RandomRangeFloat(Minimum=0, Maximum=3)) {isSetParameters=true}
   &Drag(
    Drag=&FloatFromCurve(
     FloatCurve=DataInterface() {propertyValues {Curve {keys [
         {arriveTangent=2.0771, interpMode=RCIM_Cubic, leaveTangent=2.0771, tangentMode=RCTM_User, time=0, value=0}
         {interpMode=RCIM_Cubic, tangentMode=RCTM_User, time=1, value=1}
        ]}}}
     Scale Curve=20
    )
    Use Linear Drag=true
    Use Rotational Drag=true
   )
   &SolveForcesAndVelocity(Clamp Velocity=false, Limit Acceleration=false)
   &ScaleSpriteSize(
    Scale Sprite Size Mode=::NewEnumerator0
    Uniform Scale Factor=&FloatFromCurve(
     FloatCurve=DataInterface() {propertyValues {Curve {keys [<at 0 0 RCIM_Cubic>, {arriveTangent=-2.2073, tangentMode=RCTM_User, time=1, value=1}]}}}
     Scale Curve=&FN_LOD_Distance_Remap(Start Distance=100, End Distance=500, Minimum Output=0.75, Maximum Output=1.25, Clamp Output=true)
    )
   )
  ]
  rendererClasses [&NiagaraSpriteRendererProperties]
  renderers [{rendererClass=&NiagaraSpriteRendererProperties}]
 }
}
summary {
 systemName=NS_Emote_DiamondHeart_DiamondSpawn
 emitters [
  {emitterName=Smoke, rendererClasses [&NiagaraSpriteRendererProperties]}
  {emitterName=Glitter_Small, rendererClasses [&NiagaraSpriteRendererProperties]}
  {emitterName=RoundShape, rendererClasses [&NiagaraSpriteRendererProperties]}
  {emitterName=TriangleShapes, rendererClasses [&NiagaraSpriteRendererProperties]}
  {emitterName=AirGlow, rendererClasses [&NiagaraSpriteRendererProperties]}
  {emitterName=Smoke001, rendererClasses [&NiagaraSpriteRendererProperties]}
  {emitterName=Sparkles, rendererClasses [&NiagaraSpriteRendererProperties]}
 ]
 userVariables [
  User.bIsFrontEnd {type=NiagaraBool}
  User.bIsFrontEndPreview {type=NiagaraBool}
  User.Offset {type=Vector3f, description="", defaultValue {struct=Vector3f, value=<0 0 0>}}
  User.PhysicalSurface {type=NiagaraInt32}
  User.VariantPicker {type=NiagaraFloat, defaultValue=0, description=""}
 ]
}
