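# Niagara system for the DiamondHeart emote that places diamond glint meshes and blue glow sprites on skeletal mesh bones.
# Both emitters burst System.ParticleNumber particles, using each partition index to look up bone (System.BoneAtlas), per-particle tuning and staggered starting age.
# DiamondGlints follows bone rotation with varied meshes, feeding an age curve to material parameters.
# Glow001 pulses alpha with an age-faded sine and passes bone up-vectors to the material.
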NiagaraSystem /BRCosmetics/Animation/Game/MainPlayer/Emotes/DiamondHeart/FX/NS_Emote_DiamondHeart_Glow
dependencies {
 usedDataInterfaces [
  &NiagaraDataInterfaceArrayFloat
  &NiagaraDataInterfaceArrayFloat3
  &NiagaraDataInterfaceArrayInt32
  &NiagaraDataInterfaceCurve
  &NiagaraDataInterfaceArrayFloat2
  &NiagaraDataInterfaceSkeletalMesh
 ]
 usedDynamicInputs [&SelectVector2DFromArray]
 usedModules [
  &self:SystemSpawnScript.SystemScriptSource.SystemScriptGraph.NiagaraNodeAssignment_0.SetVariables_#1
  &self:SystemSpawnScript.SystemScriptSource.SystemScriptGraph.NiagaraNodeAssignment_1.SetVariables_#2
  &self:SystemSpawnScript.SystemScriptSource.SystemScriptGraph.NiagaraNodeAssignment_2.SetVariables_#3
  &SystemState
 ]
}
emitters {
 DiamondGlints {
  emitterUpdateScript [
   &EmitterState(
    Inactive Response=::NewEnumerator1
    Life Cycle Mode=::NewEnumerator1
    Loop Behavior=::NewEnumerator1
    Loop Duration Mode=::NewEnumerator1
    UseLoopDelay=false
   )
   &SpawnBurst_Instantaneous(Spawn Count=$System.ParticleNumber, Use Spawn Probability=false, UseLoopCountLimit=false)
   &SpawnRate(SpawnRate=0.75, Use Spawn Probability=false) {enabled=false}
  ]
  particleSpawnScript [
   &PartitionParticles(Number Of Partitions=$System.ParticleNumber)
   &InitializeParticle(
    Lifetime=1.5
    Mesh Scale Mode=::NewEnumerator4
    Sprite Size=<70 70>
    Sprite Size Mode=::NewEnumerator4
    UsePositionOffset=true
    Write Material Random=true
   )
   /SetVariables(Particles.Age:NiagaraFloat=&SelectFloatFromArray(
     Direct Array Index=$Particles.PartitionParticles.Partition
     Float Selection Array=DataInterface() {propertyValues {FloatData [0,0,-2,0,-1,-1,-0.5,-0.5,-0.5,-1,-1.5,-1,-0.75]}}
    )) {isSetParameters=true, module=SetVariables_#4}
   /SetVariables(Particles.Scale=&SelectVectorFromArray(
     Direct Array Index=$Particles.PartitionParticles.Partition
     Vector Selection Array=DataInterface() {propertyValues {FloatData [
        <1 1 1.2>
        <1.4 1.4 1.8>
        <0.5 1 0.8>
        <0.8 0.8 0.7>
        <0.9 0.9 0.8>
        <1 1 1>
        <1 1 1.25>
        <1 1 1>
        <0.9 1 1>
        <1 1 1>
        <1 1 1>
        <0.8 0.8 0.7>
        <0.8 0.8 0.7>
        <1 1 1>
        <0.7 0.7 0.7>
        <0 0 0>
        <0 0 0>
        <0 0 0>
        <0 0 0>
        <0.4 0.4 0.35>
       ]}}
    )) {isSetParameters=true}
   /SetVariables(Particles.MeshIndex=&SelectIntFromArray(
     Direct Array Index=$Particles.PartitionParticles.Partition
     Int Selection Array=DataInterface() {propertyValues {IntData [0,0,2,0,0,0,3,1,1,2,0,0,2,3,2,5,5,5,5,1,5]}}
    )) {isSetParameters=true}
   &CameraOffset()
  ]
  particleUpdateScript [
   &ParticleState(Kill Particles When Lifetime Has Elapsed=false, Let Infinitely Lived Particles Die When Emitter Deactivates=true)
   &SkeletalMeshLocation(
    Bone / Socket Index=&SelectIntFromArray(Direct Array Index=$Particles.PartitionParticles.Partition, Int Selection Array=$System.BoneAtlas)
    Bone Sampling Mode=::NewEnumerator4
    MaskSpawnGroup=false
    Orientation Sampling=::NewEnumerator1
    Sampled Position Offset=&SelectVectorFromArray(
     Direct Array Index=$Particles.PartitionParticles.Partition
     Vector Selection Array=DataInterface() {propertyValues {FloatData [
        <0 0 0>
        <0 0 -1.5>
        <0 0 0>
        <0.5 0 0>
        <0 0 0>
        <0 0 0>
        <0 0 0>
        <0 0 0>
        <0 0 0>
        <0 0 0>
        <1.5 0 0>
        <0.5 0 -1>
        <0 0 0>
        <-1.5 1 1.5>
        <0 0 0>
        <0 0 0>
        <0 0 0>
        <0 0 0>
        <0 0 0>
        <0 0 0>
       ]}}
    )
    Skeletal Mesh=$System.Skeletal Mesh
   )
   /SetVariables(Particles.MeshOrientation=&MultiplyQuaternion(
     Quaternion A=$Particles.SkeletalMeshLocation.SampledRotation
     Quaternion B=&MultiplyQuaternion(
      Quaternion A=&MakeQuatFromAxisAngle(
       Axis=<1 0 0>
       AngleInDegrees=&SelectFloatFromArray(
        Direct Array Index=$Particles.PartitionParticles.Partition
        Float Selection Array=DataInterface() {propertyValues {FloatData [0,1,200,0,0,0,0,-14,-10,5,0,0,5,0,5,0,0,0,0,-15]}}
       )
      )
      Quaternion B=&MakeQuatFromAxisAngle(
       Axis=<0 0 1>
       AngleInDegrees=&SelectFloatFromArray(
        Direct Array Index=$Particles.PartitionParticles.Partition
        Float Selection Array=DataInterface() {propertyValues {FloatData [0,1,0,0,0,0,0,0,0,-90,0,0,-85,180,-85,0,0,0,0,0]}}
       )
      )
     )
    )) {isSetParameters=true}
   &DynamicMaterialParameters(
    Index 0 Param 1=&Multiply_Float(
     A=1
     B=&FloatFromCurve(FloatCurve=DataInterface() {propertyValues {Curve {keys [<at 0 0>,<at 1 1>]}}}, CurveIndex=$Particles.Age)
    )
    Index 0 Param 2=&RandomRangeFloat(Minimum=0, Maximum=1)
    Index 0 Param 3=1
    Index 0 Param 4=1
    Index 1 Param 2=30
    Module.Index0Param0Write=true
    Module.Index0Param1Write=true
    Module.Index0Param2Write=true
    Module.Index0Param3Write=true
    Module.Index1Param0Write=true
    Module.Index1Param1Write=true
    Module.Index1Param2Write=true
    Module.Index1Param3Write=true
    Write Parameter Index 1=true
   )
  ]
  rendererClasses [&NiagaraMeshRendererProperties]
  renderers [{rendererClass=&NiagaraMeshRendererProperties, rendererIndex=0}]
 }
 Glow001 {
  emitterUpdateScript [
   &EmitterState(
    Inactive Response=::NewEnumerator1
    Life Cycle Mode=::NewEnumerator1
    Loop Behavior=::NewEnumerator1
    Loop Duration Mode=::NewEnumerator1
    UseLoopDelay=false
   )
   &SpawnBurst_Instantaneous(Spawn Count=$System.ParticleNumber, Use Spawn Probability=false, UseLoopCountLimit=false)
   &SpawnRate(SpawnRate=0.75, Use Spawn Probability=false) {enabled=false}
  ]
  particleSpawnScript [
   &PartitionParticles(Number Of Partitions=$System.ParticleNumber)
   &InitializeParticle(
    Color=<rgba 0.2091 0.4702 1 1>
    Lifetime=1.5
    Position Offset=<0 -10 0>
    Sprite Size Mode=::NewEnumerator3
    Uniform Sprite Size=&SelectFloatFromArray(
     Direct Array Index=$Particles.PartitionParticles.Partition
     Float Selection Array=DataInterface() {propertyValues {FloatData [40,50,30,35,30,25,55,50,50,30,30,30,30,40,30,12,14,20,15,20,20]}}
    )
    UsePositionOffset=true
    Write Material Random=true
   )
   /SetVariables(Particles.Age:NiagaraFloat=&SelectFloatFromArray(
     Direct Array Index=$Particles.PartitionParticles.Partition
     Float Selection Array=DataInterface() {propertyValues {FloatData [0,0,-2,0,-1,-1,-0.5,-0.5,-0.5,-1,-1.5,-1,-0.75,-0.5,-1,-1.5,-2,0,0,-1,-1.75]}}
    )) {isSetParameters=true, module=SetVariables_#4}
   &CameraOffset(Camera Offset Amount=4)
  ]
  particleUpdateScript [
   &SkeletalMeshLocation(
    Bone / Socket Index=&SelectIntFromArray(Direct Array Index=$Particles.PartitionParticles.Partition, Int Selection Array=$System.BoneAtlas)
    Bone Sampling Mode=::NewEnumerator4
    MaskSpawnGroup=false
    Orientation Sampling=::NewEnumerator1
    Sampled Orientation Coordinate Space=::World
    Sampled Position Offset=&SelectVectorFromArray(
     Direct Array Index=$Particles.PartitionParticles.Partition
     Vector Selection Array=DataInterface() {propertyValues {FloatData [<0 0 0>, <0 0 0>, <0 0 0>, <0 0 0>, <0 0 0>, <0 0 0>, <0 0 0>, <0 0 0>, <0 0 0>, <0 0 0>, <0 0 0>]}}
    )
    Skeletal Mesh=$System.Skeletal Mesh
   )
   &ParticleState(Kill Particles When Lifetime Has Elapsed=false)
   &ScaleSpriteSize(
    Uniform Curve Index=$Particles.Age
    Uniform Curve Sprite Scale=DataInterface() {propertyValues {Curve {keys [{tangentMode=RCTM_User, time=0, value=0}, {tangentMode=RCTM_User, time=0.5, value=1}]}}}
   ) {enabled=false}
   &ScaleColor(
    Scale Alpha=&Multiply_Float(
     A=&Sine(
      Normalized Angle=&Add_Float(A=$Emitter.Age, B=&RandomRangeFloat(Minimum=0, Maximum=3))
      Period=6
      Scale=1
      Bias=2
     )
     B=&FloatFromCurve(
      FloatCurve=DataInterface() {propertyValues {Curve {keys [{tangentMode=RCTM_User, time=0, value=0}, {tangentMode=RCTM_User, time=0.5, value=1}]}}}
      CurveIndex=$Particles.Age
     )
    )
    ScaleA=true
    ScaleRGB=true
   )
   &DynamicMaterialParameters(
    Index 0 Param 3=&RandomRangeFloat(Minimum=0, Maximum=1)
    Index 0 Param 4=&RandomRangeFloat(Minimum=0, Maximum=1)
    Index 1 Param 2=30
    Module.Index0Param0Write=true
    Module.Index0Param1Write=true
    Module.Index0Param2Write=true
    Module.Index0Param3Write=true
    Module.Index1Param0Write=true
    Module.Index1Param1Write=true
    Module.Index1Param2Write=true
    Module.Index1Param3Write=true
    Write Parameter Index 1=true
   ) {enabled=false}
   &DynamicMaterialParameters(
    Index 0 Param 1=&MakeFloatFromVector(
     Vector=&RotateVectorByQuaternion(Vector To Rotate=<0 0 1>, Quaternion=$Particles.SkeletalMeshLocation.SampledRotation)
     Channel=::NewEnumerator0
    )
    Index 0 Param 2=&MakeFloatFromVector(
     Vector=&RotateVectorByQuaternion(Vector To Rotate=<0 0 1>, Quaternion=$Particles.SkeletalMeshLocation.SampledRotation)
     Channel=::NewEnumerator1
    )
    Index 0 Param 3=&MakeFloatFromVector(
     Vector=&RotateVectorByQuaternion(Vector To Rotate=<0 0 1>, Quaternion=$Particles.SkeletalMeshLocation.SampledRotation)
     Channel=::NewEnumerator2
    )
    Index 0 Param 4=&SelectFloatFromArray(
     Direct Array Index=$Particles.PartitionParticles.Partition
     Float Selection Array=DataInterface() {propertyValues {FloatData [1,1,1,1,1,1,70,100,60,1,1,1,1,1,1,1,1,1,1,1,1]}}
    )
    Index 1 Param 2=30
    Module.Index0Param0Write=true
    Module.Index0Param1Write=true
    Module.Index0Param2Write=true
    Module.Index0Param3Write=true
    Module.Index1Param0Write=true
    Module.Index1Param1Write=true
    Module.Index1Param2Write=true
    Module.Index1Param3Write=true
    Write Parameter Index 1=true
   ) {module=DynamicMaterialParameters001}
  ]
  rendererClasses [&NiagaraSpriteRendererProperties]
  renderers [{rendererClass=&NiagaraSpriteRendererProperties}]
 }
}
summary {
 systemName=NS_Emote_DiamondHeart_Glow
 emitters [
  {emitterName=DiamondGlints, rendererClasses [&NiagaraMeshRendererProperties]}
  {emitterName=Glow001, rendererClasses [&NiagaraSpriteRendererProperties]}
 ]
 userVariables [
  User.Offset {type=Vector3f, description="", defaultValue {struct=Vector3f, value=<0 0 0>}}
  User.ReliableElkVariant {type=NiagaraBool}
  User.Size {type=NiagaraFloat, defaultValue=1, description=""}
 ]
}
